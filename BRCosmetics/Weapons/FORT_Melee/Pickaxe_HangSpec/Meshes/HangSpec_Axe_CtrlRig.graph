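# Control Rig for the HangSpec pickaxe, driving its handle and skin-chain bones from animation controls.
# Execution loops over a constant bone list, pairs each bone by index with a control, and sets its transform by projecting its initial offset onto that control's current transform.
# Controls form one serial chain from root through bottom handle and skin chain to top handle, although the skin-chain bones are siblings under the handle bone.
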ControlRigBlueprint /BRCosmetics/Weapons/FORT_Melee/Pickaxe_HangSpec/Meshes/HangSpec_Axe_CtrlRig
graphs [
 RigVMFunctionLibrary {links [], nodes []}
 RigVMModel {nodes [
   RigUnit_BeginExecution = &RigUnit_BeginExecution()
   ItemArray = RigVMDispatchNode(Value="((Type=Bone,Name=\"root\"),(Type=Bone,Name=\"offset\"),(Type=Bone,Name=\"handle\"),(Type=Bone,Name=\"btm_handle_btm\"),(Type=Bone,Name=\"skin_chain_01\"),(Type=Bone,Name=\"skin_chain_02\"),(Type=Bone,Name=\"skin_chain_03\"),(Type=Bone,Name=\"skin_chain_04\"),(Type=Bone,Na ...(truncated)") {template="DISPATCH_RigVMDispatch_Constant(io Value)"}
   For_Each = RigVMDispatchNode(Array=@ItemArray.Value, ExecuteContext=@RigUnit_BeginExecution.ExecutePin) {template="DISPATCH_RigVMDispatch_ArrayIterator(in Array,out Element,out Index,out Count,out Ratio)"}
   ItemArray_1 = RigVMDispatchNode(Value="((Type=Control,Name=\"root_ctrl\"),(Type=Control,Name=\"offset_ctrl\"),(Type=Control,Name=\"handle_ctrl\"),(Type=Control,Name=\"btm_handle_btm_ctrl\"),(Type=Control,Name=\"skin_chain_01_ctrl\"),(Type=Control,Name=\"skin_chain_02_ctrl\"),(Type=Control,Name=\"skin_chain_ ...(truncated)") {template="DISPATCH_RigVMDispatch_Constant(io Value)"}
   n1 = RigVMDispatchNode(Index=@For_Each.Index, Array=@ItemArray_1.Value) {template="DISPATCH_RigVMDispatch_ArrayGetAtIndex(in Array,in Index,out Element)"}
   > &RigUnit_SetTransform(
    Weight=1
    bInitial=False
    Item=@For_Each.Element
    Value=/Script/ControlRig.RigUnit_ProjectTransformToNewParent(bChildInitial=True, bNewParentInitial=False, bOldParentInitial=True, NewParent=@n1.Element, OldParent=@n1.Element, Child=@For_Each.Element) {is=struct RigVMUnitNode}.Transform
   )
  ]}
]
hierarchy [
 root {loc=<0 0 0>, rot=<0 0 0 1>}
 offset {parent=root, loc=<0 0 0>, rot=<0 0 -0.7071 -0.7071>}
 handle {parent=offset, loc=<0 0 0>, rot=<0 0 0 1>}
 btm_handle_btm {parent=handle, loc=<0 0 13.1859>, rot=<0.5 0.5 0.5 -0.5>}
 btm_handle_mid_lwr {parent=btm_handle_btm, loc=<-7.6699 0 0>, rot=<0 0 0 1>}
 btm_handle_mid {parent=btm_handle_mid_lwr, loc=<-6.0118 0 0>, rot=<0 0 0 1>}
 btm_handle_mid_upr {parent=btm_handle_mid, loc=<-5.5121 0 0>, rot=<0 0 0 1>}
 btm_handle_top {parent=btm_handle_mid_upr, loc=<-7.3271 0 0>, rot=<0 0 0 1>}
 skin_chain_11 {parent=handle, loc=<0 0 37.4184>, rot=<-0.5 0.5 0.5 -0.5>}
 skin_chain_10 {parent=handle, loc=<0.0004 -0.0003 34.8707>, rot=<-0.4999 0.5003 0.4997 -0.5001>}
 skin_chain_09 {parent=handle, loc=<0.0011 -0.0008 32.9991>, rot=<-0.4997 0.5005 0.4995 -0.5003>}
 skin_chain_08 {parent=handle, loc=<0.0018 -0.0014 31.0329>, rot=<-0.4999 0.5002 0.4998 -0.5001>}
 skin_chain_07 {parent=handle, loc=<0.0024 -0.0018 29.0733>, rot=<0.5004 -0.4997 -0.5003 0.4996>}
 skin_chain_06 {parent=handle, loc=<0.0026 -0.002 27.1155>, rot=<-0.5009 0.4992 0.5009 -0.4991>}
 skin_chain_05 {parent=handle, loc=<0.0025 -0.0019 25.0909>, rot=<-0.5007 0.4992 0.5008 -0.4993>}
 skin_chain_04 {parent=handle, loc=<0.0019 -0.0014 23.0681>, rot=<-0.5002 0.4997 0.5004 -0.4998>}
 skin_chain_03 {parent=handle, loc=<0.0011 -0.0008 21.0427>, rot=<-0.5 0.4998 0.5002 -0.5>}
 skin_chain_02 {parent=handle, loc=<0.0004 -0.0003 18.813>, rot=<-0.5 0.4999 0.5001 -0.5>}
 skin_chain_01 {parent=handle, loc=<0 0 16.227>, rot=<-0.5 0.5 0.5 -0.5>}
 top_handle_top {parent=handle, loc=<0 0 40.5461>, rot=<0.5 0.5 0.5 -0.5>}
 top_handle_mid_upr {parent=top_handle_top, loc=<7.7873 0 0>, rot=<0 0 0 1>}
 top_handle_mid {parent=top_handle_mid_upr, loc=<5.755 0 0>, rot=<0 0 0 1>}
 top_handle_mid_lwr {parent=top_handle_mid, loc=<5.2529 0 0>, rot=<0 0 0 1>}
 top_handle_btm {parent=top_handle_mid_lwr, loc=<7.5889 0 0>, rot=<0 0 0 1>}
 root_ctrl {controlType=EULER_TRANSFORM, type=Control, loc=<0 0 0>, rot=<0 0 0 1>}
 offset_ctrl {controlType=EULER_TRANSFORM, parent=root_ctrl, type=Control, loc=<0 0 0>, rot=<0 0 0 1>}
 handle_ctrl {controlType=EULER_TRANSFORM, parent=offset_ctrl, type=Control, loc=<0 0 0>, rot=<0 0 0 1>}
 btm_handle_btm_ctrl {controlType=EULER_TRANSFORM, parent=handle_ctrl, type=Control, loc=<0 0 0>, rot=<0 0 0 1>}
 skin_chain_01_ctrl {controlType=EULER_TRANSFORM, parent=btm_handle_btm_ctrl, type=Control, loc=<0 0 0>, rot=<0 0 0 1>}
 skin_chain_02_ctrl {controlType=EULER_TRANSFORM, parent=skin_chain_01_ctrl, type=Control, loc=<0 0 0>, rot=<0 0 0 1>}
 skin_chain_03_ctrl {controlType=EULER_TRANSFORM, parent=skin_chain_02_ctrl, type=Control, loc=<0 0 0>, rot=<0 0 0 1>}
 skin_chain_04_ctrl {controlType=EULER_TRANSFORM, parent=skin_chain_03_ctrl, type=Control, loc=<0 0 0>, rot=<0 0 0 1>}
 skin_chain_05_ctrl {controlType=EULER_TRANSFORM, parent=skin_chain_04_ctrl, type=Control, loc=<0 0 0>, rot=<0 0 0 1>}
 skin_chain_06_ctrl {controlType=EULER_TRANSFORM, parent=skin_chain_05_ctrl, type=Control, loc=<0 0 0>, rot=<0 0 0 1>}
 skin_chain_07_ctrl {controlType=EULER_TRANSFORM, parent=skin_chain_06_ctrl, type=Control, loc=<0 0 0>, rot=<0 0 0 1>}
 skin_chain_08_ctrl {controlType=EULER_TRANSFORM, parent=skin_chain_07_ctrl, type=Control, loc=<0 0 0>, rot=<0 0 0 1>}
 skin_chain_09_ctrl {controlType=EULER_TRANSFORM, parent=skin_chain_08_ctrl, type=Control, loc=<0 0 0>, rot=<0 0 0 1>}
 skin_chain_10_ctrl {controlType=EULER_TRANSFORM, parent=skin_chain_09_ctrl, type=Control, loc=<0 0 0>, rot=<0 0 0 1>}
 skin_chain_11_ctrl {controlType=EULER_TRANSFORM, parent=skin_chain_10_ctrl, type=Control, loc=<0 0 0>, rot=<0 0 0 1>}
 top_handle_top_ctrl {controlType=EULER_TRANSFORM, parent=skin_chain_11_ctrl, type=Control, loc=<0 0 0>, rot=<0 0 0 1>}
]
